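# Physics asset for the Wise Master medium male backpack cosmetic, giving its dangling bottle and four knives secondary motion.
# The pack and bottle use capsule bodies, the knives use thin boxes, and small master bodies act as anchors for each group.
# Each knife is constrained to the knife master with tight swing and twist limits, while the bottle hangs from its own master with a looser, uniform cone.
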PhysicsAsset /BRCosmetics/Accessories/FORT_Backpacks/M_MED_Wise_Master_Backpack/M_MED_Wise_Master_Pack_PhysicsAsset
bodies {
 dyn_bottle [{shapeType=Capsule, length=10.9781, radius=7.1795, center=<-1.249 0.0123 -1.5549>, rotation=<rot 90 0 0>}]
 dyn_bottle_mstr [{shapeType=Box, extentX=1, extentY=1, extentZ=1}]
 dyn_knives_a [{shapeType=Box, extentX=19.885, extentY=2.0613, extentZ=3.8212, center=<5.1623 -0.6433 -0.1538>}]
 dyn_knives_b [{shapeType=Box, extentX=19.9196, extentY=2.0055, extentZ=3.821, center=<5.0717 -0.1455 0.228>}]
 dyn_knives_c [{shapeType=Box, extentX=19.8923, extentY=2.0346, extentZ=3.8244, center=<5.0299 0.3792 0.312>}]
 dyn_knives_d [{shapeType=Box, extentX=19.8893, extentY=1.1, extentZ=3.8067, center=<5.2066 0 0>}]
 dyn_knives_mstr [{shapeType=Capsule, length=2, radius=2, center=<-7.7957 1.8937 3.8558>, rotation=<rot -1.3056 -90 30.8303>}]
 pack [{shapeType=Capsule, length=23.8606, radius=1.185, center=<-1.0228 -4.4187 -0.0483>, rotation=<rot 6.0888 -90 20.7732>}]
}
constraints [
 {bone1Name=dyn_knives_a, bone2Name=dyn_knives_mstr, swing1LimitDegrees=1, swing2LimitDegrees=10, twistLimitDegrees=2.5}
 {bone1Name=dyn_knives_b, bone2Name=dyn_knives_mstr, swing1LimitDegrees=1, swing2LimitDegrees=10, twistLimitDegrees=2.5}
 {bone1Name=dyn_knives_c, bone2Name=dyn_knives_mstr, swing1LimitDegrees=1, swing2LimitDegrees=10, twistLimitDegrees=2.5}
 {bone1Name=dyn_knives_d, bone2Name=dyn_knives_mstr, swing1LimitDegrees=1, swing2LimitDegrees=10, twistLimitDegrees=2.5}
 {bone1Name=dyn_bottle, bone2Name=dyn_bottle_mstr, swing1LimitDegrees=30, swing2LimitDegrees=30, twistLimitDegrees=30}
]
