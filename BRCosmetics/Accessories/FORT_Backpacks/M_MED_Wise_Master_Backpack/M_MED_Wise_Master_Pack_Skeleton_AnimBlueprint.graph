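# Animation Blueprint for the Wise Master backpack skeleton that adjusts the pack bone and adds physics-driven secondary motion.
# The input pose passes through chained Modify Bone nodes that add a roll in the front end, a small constant offset, and a yaw while aiming down sights.
# Targeting events switch ADS_Alpha between 0 and 1 to drive that yaw.
# A RigidBody node then simulates the pack's physics asset with overridden gravity and damped component motion.
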AnimBlueprint /BRCosmetics/Accessories/FORT_Backpacks/M_MED_Wise_Master_Backpack/M_MED_Wise_Master_Pack_Skeleton_AnimBlueprint
target_skeleton=&/M_MED_Wise_Master_Pack_Skeleton
graphs [
 AnimGraph {nodes [
   AnimGraphNode_SubInput_0 = Input Pose()
   n1 = "Transform (Modify) Bone - Bone: pack"!(
    ComponentPose=Local To Component(LocalPose=@AnimGraphNode_SubInput_0)
    bAlphaBoolEnabled:bool=Get bIsInFrontEnd()
    Translation="0,0,0"
    Rotation="0,0,12"
    Scale
   ) -> (Pose) {
    alphaInputType=Bool
    rotationMode=BMM_Additive
    translationMode=BMM_Additive
    alphaBoolBlend {blendOption=Linear}
    boneToModify {boneName=pack}
    rotation {roll=12}
   }
   n2 = "Transform (Modify) Bone - Bone: pack"!(ComponentPose=@n1, Alpha:float, Translation="0,-1.4,0", Rotation="0,0,0", Scale) -> (Pose) {
    rotationMode=BMM_Additive
    rotationSpace=BCS_ParentBoneSpace
    translationMode=BMM_Additive
    alphaBoolBlend {blendOption=Linear}
    boneToModify {boneName=pack}
    translation {y=-1.4}
   }
   n3 = "Transform (Modify) Bone - Bone: pack"(ComponentPose=@n2, Alpha=Get ADS_Alpha() -> (ADS_Alpha:double), Translation="0,0,0", Rotation="0,-12.5,0") {
    rotationMode=BMM_Additive
    rotationSpace=BCS_ParentBoneSpace
    alphaBoolBlend {blendOption=Linear}
    boneToModify {boneName=pack}
    rotation {yaw=-12.5}
   }
   n4 = RigidBody(ComponentPose=@n3) {
    bClampLinearTranslationLimitToRefPose=true
    bOverrideWorldGravity=true
    overridePhysicsAsset=&/M_MED_Wise_Master_Pack_PhysicsAsset
    alphaBoolBlend {blendOption=Linear}
    componentAppliedLinearAccClamp=<3000 3000 3000>
    componentLinearAccScale=<0.6 0.6 0.6>
    componentLinearVelScale=<0.6 0.6 0.6>
    overrideWorldGravity {z=-980}
   }
   Output Pose(Result=Component To Local(ComponentPose=@n4)) {layerGroup=None}
  ]}
 EventGraph {nodes [
   Event OnBeginTargeting() {is=title K2Node_Event}
   > Set ADS_Alpha(ADS_Alpha:double=1) -> (Output_Get:double)
   Event OnEndTargeting() {is=title K2Node_Event}
   > Set ADS_Alpha(ADS_Alpha=0)
   TryGetPawnOwner()
   Event BlueprintUpdateAnimation()
  ]}
]
variables [ADS_Alpha {type=number}]
